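; Battle Pass Season 21 Week 5 quest actor for an interactable catacomb mirror that redirects a Niagara beam.
; Local interaction switches the mirror and beam visuals from pre to post state, and construction previews either state.
; The actor faces the midpoint of the incoming beam location and outgoing end point.
; SetupLerp, uncalled here, starts timers that interpolate the pre-interaction beam and mirror rotation toward the post state.
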(event EventBeginPlay
  (|CallParentFunction)
  (CallFunction|SetModelDirection false)
  (Variables|Default|SetQuestActive false))

(event Custom|MoveBeam
  (CallFunction|LerpBeam))

(event Custom|EndMoveBeam
  (CallFunction|CleanUpTimers))

(event Custom|CleanUpTimers
  (Utilities|Time|ClearandInvalidateTimerbyHandle (Variables|Default|GetMoveBeamTimer))
  (Utilities|Time|ClearandInvalidateTimerbyHandle (Variables|Default|GetMoveBeamIntervalTimer)))

(event EventEndPlay
  (params (EndPlayReason :type EEndPlayReason%20Enum :default "Destroyed"))
  (CallFunction|CleanUpTimers)
  (|CallParentFunction EndPlayReason))

(fn BlueprintOnLocalInteract
  (params (InteractingPawn :type Fort%20Player%20Pawn%20Object%20Reference) (InteractionBeingAttempted :type EInteractionBeingAttempted%20Enum :default "FirstInteraction"))
  (bind _returnvalue (|CallParentFunction InteractingPawn InteractionBeingAttempted))
  (CallFunction|SetModelDirection true)
  (CallFunction|SetBeamVisibility true)
  (return _returnvalue))

(fn LerpBeam
  (bind _movetime (Variables|Default|GetMoveTime))
  (bind _movebeamtimer (Variables|Default|GetMoveBeamTimer))
  (bind _returnvalue (Utilities|Time|GetTimerElapsedTimebyHandle _movebeamtimer))
  (bind _returnvalue_2 (/ _returnvalue _movetime))
  (bind _endpostinteractionvfxpos (Variables|AVariables:Post|GetEndPostInteractionVfxPos))
  (bind _endpreinteractionvfxpos (Variables|AVariables:Pre|GetEndPreInteractionVfxPos))
  (bind _returnvalue_3 (Math|Vector|Lerp(Vector) _endpreinteractionvfxpos _endpostinteractionvfxpos _returnvalue_2))
  (Niagara|SetNiagaraVariable(Position) (Variables|Default|GetBeam_Niagara_Pre) "StartPos" (Math|Vector|Lerp(Vector) (Variables|AVariables:Pre|GetStartPreInteractionVfxPos) (Variables|AVariables:Post|GetStartPostInteractionVfxPos) _returnvalue_2))
  (Niagara|SetNiagaraVariable(Position) (Variables|Default|GetBeam_Niagara_Pre) "EndPos" _returnvalue_3)
  (Transformation|SetActorRotation (Math|Rotator|FindLookatRotation (Transformation|GetWorldLocation (Variables|Transformation|GetRootComponent)) (/ (+ (Variables|AVariables|GetIncomingBeamLoc) _returnvalue_3) "2.000000,2.000000,2.000000"))))

(fn SetBeamLocations
  (Niagara|SetNiagaraVariable(Position) (Variables|Default|GetBeam_Niagara_Post) "EndPos" (Variables|AVariables:Post|GetEndPostInteractionVfxPos))
  (Niagara|SetNiagaraVariable(Position) (Variables|Default|GetBeam_Niagara_Post) "StartPos" (Variables|AVariables:Post|GetStartPostInteractionVfxPos))
  (Niagara|SetNiagaraVariable(Position) (Variables|Default|GetBeam_Niagara_Pre) "EndPos" (Variables|AVariables:Pre|GetEndPreInteractionVfxPos))
  (Niagara|SetNiagaraVariable(Position) (Variables|Default|GetBeam_Niagara_Pre) "StartPos" (Variables|AVariables:Pre|GetStartPreInteractionVfxPos)))

(fn SetBeamVisibility
  (params (Post :type Boolean))
  (Rendering|SetVisibility (Variables|Default|GetBeam_Niagara_Pre) (not Post) false)
  (Rendering|SetVisibility (Variables|Default|GetBeam_Niagara_Post) Post false))

(fn SetModelDirection
  (params (Post :type Boolean))
  (if Post
    (Transformation|SetRelativeLocationAndRotation (Variables|Default|GetScene_MirrorParent) (Variables|AVariables:Post|GetPostInteractionPos) (Math|Conversions|RotationFromXVector (Variables|AVariables:Post|GetPostInteractionRot)) false false)
    (Variables|AVariables|SetCurrentEndPos (Variables|AVariables:Post|GetEndPostInteractionVfxPos))
    (else
      (Transformation|SetRelativeLocationAndRotation (Variables|Default|GetScene_MirrorParent) (Variables|AVariables:Pre|GetPreInteractionPos) (Math|Conversions|RotationFromXVector (Variables|AVariables:Pre|GetPreInteractionRot)) false false)
      (Variables|AVariables|SetCurrentEndPos (Variables|AVariables:Pre|GetEndPreInteractionVfxPos))))
  (Transformation|SetActorRotation (Math|Rotator|FindLookatRotation (Transformation|GetWorldLocation (Variables|Transformation|GetRootComponent)) (/ (+ (Variables|AVariables|GetIncomingBeamLoc) (Variables|AVariables|GetCurrentEndPos)) "2.000000,2.000000,2.000000"))))

(fn SetupLerp
  (bind _returnvalue (Utilities|Time|SetTimerbyEvent (EventDispatchers|CreateEvent :!event MoveBeam) (Variables|Default|GetMoveInterval) true))
  (Variables|Default|SetMoveBeamIntervalTimer _returnvalue)
  (bind _returnvalue_2 (Utilities|Time|SetTimerbyEvent (EventDispatchers|CreateEvent :!event EndMoveBeam) (Variables|Default|GetMoveTime)))
  (Variables|Default|SetMoveBeamTimer _returnvalue_2))

(fn ConstructionScript
  (bind _previewpost (Variables|AVariables|GetPreviewPost))
  (|CallParentFunction)
  (CallFunction|SetBeamLocations)
  (CallFunction|SetBeamVisibility _previewpost)
  (CallFunction|SetModelDirection _previewpost))
