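; BP_S21_W05_CatacombMirror_Source is a Season 21 Week 5 narrative quest actor, apparently the source mirror in a catacomb mirror interaction.
; After calling its parent, the construction script writes StartPos and EndPos position parameters on separate pre-interaction and post-interaction Niagara components from their own position variables.
; BeginPlay only calls the parent, so any interaction or state-switching behavior is not defined in this graph.
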(event EventBeginPlay
  (|CallParentFunction))

(fn ConstructionScript
  (|CallParentFunction)
  (Niagara|SetNiagaraVariable(Position) (Variables|Default|GetPreInteraction_Niagara) "EndPos" (Variables|Default|GetEndPreInteractionVfxPos))
  (Niagara|SetNiagaraVariable(Position) (Variables|Default|GetPreInteraction_Niagara) "StartPos" (Variables|Default|GetStartPreInteractionVfxPos))
  (Niagara|SetNiagaraVariable(Position) (Variables|Default|GetPostInteraction_Niagara) "EndPos" (Variables|Default|GetEnd%20PostInteractionVfxPos))
  (Niagara|SetNiagaraVariable(Position) (Variables|Default|GetPostInteraction_Niagara) "StartPos" (Variables|Default|GetStartPostInteractionVfxPos)))
