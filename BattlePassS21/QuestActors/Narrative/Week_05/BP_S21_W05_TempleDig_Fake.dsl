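; Quest prop actor for the Season 21 Week 5 narrative quest, apparently a decoy temple dig site judging by its name.
; After interaction, it runs the parent's VFX and audio response, then gets its quest icon component and hides that icon on the map, minimap and in the world.
; BeginPlay and the construction script only call the parent, so all other behavior is inherited.
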(event VisualState|VFX&Audio|EventPostInteractionVFX&Audio
  (params (PlayerController :type Fort%20Player%20Controller%20Object%20Reference))
  (bind _questiconcomponent (Variables|BuildingGameplayActorPropQuest|GetQuestIconComponent))
  (|CallParentFunction 0)
  (MiniMapData|SetIconVisibilityonMap _questiconcomponent false)
  (MiniMapData|SetIconVisibilityonMiniMap _questiconcomponent false)
  (MiniMapData|SetIconVisibilityinWorld _questiconcomponent false))

(event EventBeginPlay
  (|CallParentFunction))

(fn ConstructionScript
  (|CallParentFunction))
